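# Niagara system for the swing trail on the Cavern Male pickaxe, drawing one short-lived, faint pale-blue ribbon.
# A single FN_AnimTrail_V4 emitter uses shared FN anim-trail modules to spawn, initialize from a bone or socket index, and update particles.
# ScaleColor applies an alpha curve that rises quickly, then fades out over each particle's life.
# The system exposes skeletal mesh and ribbon ID offset user parameters.
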NiagaraSystem /BRCosmetics/Weapons/FORT_Melee/Pickaxe_Cavern_Male/FX/NS_Pickaxe_Cavern_Male_AnimTrail
dependencies {usedDataInterfaces [&NiagaraDataInterfaceCurve], usedDynamicInputs [], usedModules [&SystemState]}
emitters {FN_AnimTrail_V4 {
  emitterUpdateScript [
   &EmitterState()
   /Game/Effects/Niagara/Modules/AnimTrail/FN_SpawnAnimTrail/FN_SpawnAnimTrail(Number of Anim Trails=1, SpawnRate=60)
  ]
  particleSpawnScript [
   /Game/Effects/Niagara/Modules/AnimTrail/FN_InitializeAnimTrail/FN_InitializeAnimTrailBoneIndecies(First Bone or Socket Index=1)
   /Game/Effects/Niagara/Modules/AnimTrail/FN_InitializeAnimTrail/FN_InitializeAnimTrail(AnimTrailWidth=1.05, Color=<rgba 0.6302 0.7223 1 0.2>, Lifetime=0.15)
   /SetVariables(Particles.RibbonWidth=&MakeCustomFloatFromBool(Bool=$Particles.Locked, True Float=$Particles.RibbonWidth, False Float=$Particles.RibbonWidth)) {isSetParameters=true}
  ]
  particleUpdateScript [
   /Game/Effects/Niagara/Modules/AnimTrail/FN_UpdateAnimTrail/FN_UpdateAnimTrail()
   &ScaleColor(
    Scale Alpha=&FloatFromCurve(FloatCurve=DataInterface() {propertyValues {Curve {keys [<at 0 0 RCIM_Cubic>,<at 0.15 1 RCIM_Cubic>,<at 1 0>]}}})
    ScaleA=true
    ScaleRGB=true
   )
  ]
  rendererClasses [&NiagaraRibbonRendererProperties]
  renderers [{rendererClass=&NiagaraRibbonRendererProperties}]
 }}
summary {
 systemName=NS_Pickaxe_Cavern_Male_AnimTrail
 emitters [{emitterName=FN_AnimTrail_V4, rendererClasses [&NiagaraRibbonRendererProperties]}]
 userVariables [
  User.RibbonIDOffset {type=NiagaraInt32}
  User.SkeletalMesh {
   type=NiagaraDataInterfaceSkeletalMesh
   defaultValue {
    struct=NiagaraExt_VariableValue_DataInterface
    value {dataInterface=&self:NiagaraDataInterfaceSkeletalMesh_0}
   }
  }
 ]
}
